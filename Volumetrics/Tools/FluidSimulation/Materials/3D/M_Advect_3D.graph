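# Unlit material for the advection pass of a 3D fluid simulation on volumes tiled into 2D textures.
# It traces back along Velocity to sample Source, then adds spherical emission, clamps and dampens.
# IsVelocity selects velocity processing (boundary inflow dampening, panned curl noise, buoyancy from non-advected Color alpha) or color processing (erosion, optional edge mask).
# AlphaComposite with negative emissive allows signed values, with alpha inverted by OneMinus.
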Material /Volumetrics/Tools/FluidSimulation/Materials/3D/M_Advect_3D
graph {
 properties {
  bAllowNegativeEmissiveColor=true
  blendMode=BLEND_AlphaComposite
  pixelDepthOffsetMode=PDOM_Legacy
  shadingModel=MSM_Unlit
  usedShadingModels=Unlit
 }
 nodes [
  n1 = ComponentMask(None=VectorParameter() {parameterName=XYFrames, defaultValue=<rgba 16 16 0 0>}.RGB) {g=true, r=true}
  n2 = MaterialFunctionCall(Speed=ScalarParameter() {defaultValue=0.5, parameterName=Constant Curl Pan Speed}) {materialFunction=&TimeWithSpeedVariable}
  n3 = VectorParameter() {parameterName=Pan Dir, defaultValue=<rgba 0 -0.05 0.1 0>}
  n4 = Multiply(
   A=Add(
    A=MaterialFunctionCall!(UVs, XY Frames=@n1) {materialFunction=/Volumetrics/Content/VolumeTextures/MaterialFunctions/EncodeVolumeCoordinates}.Result
    B=Multiply(A=@n2.Frac Time, B=@n3.RGB)
   )
   B=ScalarParameter() {defaultValue=1, parameterName=Constant Curl Tiling}
  )
  n5 = TextureObjectParameter() {desc=Velocity, parameterName=Source, samplerType=SAMPLERTYPE_LinearColor, texture=/Volumetrics/Content/VolumeTextures/Textures/RT_Default}
  n6 = ComponentMask(None=VectorParameter() {parameterName=XYFrames, defaultValue=<rgba 16 16 0 0>}.RGB) {g=true, r=true}
  n7 = TextureSampleParameter2D() {desc=Velocity, parameterName=Velocity, samplerType=SAMPLERTYPE_LinearColor, texture=&RT_Default}
  n8 = ComponentMask(None=@n7.RGB) {b=true, g=true, r=true}
  n9 = Divide(
   A=ScalarParameter() {defaultValue=5, desc="Default = 90 hz", parameterName=TimeStep}
   B=VectorParameter() {parameterName=FrameResolution, defaultValue=<rgba 256 256 256 0>}.RGB
  )
  n10 = MaterialFunctionCall!(
   Texture Object=@n5
   Position=Subtract(
    A=MaterialFunctionCall(XY Frames=@n6) {materialFunction=&EncodeVolumeCoordinates}.Result
    B=Multiply(A=@n8, B=@n9)
   ) {constB=0}
   Scale=Constant() {r=1}
   Num Frames=@n6
   Clamp Z
  ) {materialFunction=/Volumetrics/Content/VolumeTextures/MaterialFunctions/VolumeTextureFunction}
  n11 = MaterialFunctionCall() {materialFunction=/Volumetrics/Tools/FluidSimulation/MaterialFunctions/FluidBoundary_3D}
  n12 = Multiply(
   A=@n10.Result RGB
   B=MaterialFunctionCall(X=@n11.Negative X, Y=@n11.Negative X, Z=@n11.Negative X) {materialFunction=&MakeFloat3}.Result
  )
  n13 = Multiply(
   A=@n10.Result RGB
   B=MaterialFunctionCall(X=@n11.Positive X, Y=@n11.Positive X, Z=@n11.Positive X) {materialFunction=&MakeFloat3}.Result
  )
  n14 = Multiply(A=Add(A=Max(A=@n12) {constB=0}, B=Min(A=@n13) {constB=0}), B=ScalarParameter() {parameterName=InFlow Dampening})
  n15 = TextureSampleParameter2D() {desc=Color, parameterName=Color, samplerType=SAMPLERTYPE_LinearColor, texture=&RT_Default}
  n16 = Multiply(A=@n15.A, B=Constant3Vector() {constant=<rgba 0 0 1 0>}) {desc="Test using Non-Advected Color Temperature Advection, so Obstacle can be written to Velocity.A, in Gradient Subtract"}
  n17 = ScalarParameter() {defaultValue=5, parameterName=Temp Buoyancy}
  n18 = Add(
   A=Multiply(
    A=TextureSample(UVs=@n4) {samplerType=SAMPLERTYPE_LinearColor, texture=/Volumetrics/Content/VolumeTextures/Textures/VT_CurlNoise}.RGB
    B=ScalarParameter() {parameterName=Constant Curl}
   )
   B=Add(A=Subtract(A=@n10.Result RGB, B=@n14), B=Multiply(A=@n16, B=@n17)) {constB=0}
  )
  n19 = ScalarParameter() {defaultValue=0.1, parameterName=Color Erosion}
  n20 = Subtract(A=@n10.Result RGB, B=@n19)
  n21 = OneMinus(None=@n19)
  n22 = Min(A=Max(A=Divide(A=@n20, B=@n21)) {constB=0}, B=@n10.Result RGB)
  n23 = ComponentMask(None=VectorParameter() {parameterName=XYFrames, defaultValue=<rgba 16 16 0 0>}.RGB) {g=true, r=true}
  n24 = Multiply(
   A=SphereMask(
    A=MaterialFunctionCall(XY Frames=@n23) {materialFunction=&EncodeVolumeCoordinates}.Result
    B=VectorParameter() {parameterName=P, defaultValue=<rgba 0.5 0.5 0.5 0>}.RGB
    Radius=ScalarParameter() {defaultValue=0.05, parameterName=SourceRadius}
   ) {hardnessPercent=0}
   B=ScalarParameter() {defaultValue=1, parameterName=Emission Strength}
  )
  n25 = Multiply(A=@n24, B=Multiply(A=ScalarParameter() {defaultValue=1, parameterName=Emission Temperature}) {constB=0.01})
  n26 = Add(
   A=AppendVector(A=StaticSwitchParameter(True=@n18, False=@n22) {defaultValue=true, parameterName=IsVelocity}, B=@n10.Result Alpha)
   B=StaticSwitchParameter(
    True=MaterialFunctionCall(A=@n25) {materialFunction=&MakeFloat4}.Result
    False=MaterialFunctionCall(X=@n24, Y=@n24, Z=@n24, A=@n25) {materialFunction=&MakeFloat4}.Result
   ) {defaultValue=true, parameterName=IsVelocity}
  ) {constB=0}
  n27 = StaticSwitchParameter(
   True=ScalarParameter() {defaultValue=0.99, parameterName=Velocity Dampening}
   False=ScalarParameter() {defaultValue=1, parameterName=Color Dampening}
  ) {defaultValue=true, parameterName=IsVelocity}
  n28 = ScalarParameter() {defaultValue=0.95, parameterName=Temperature Dampening}
  n29 = VectorParameter() {parameterName=XYFrames, defaultValue=<rgba 16 16 0 0>}
  n30 = MaterialFunctionCall!(
   UVW=MaterialFunctionCall(XY Frames=ComponentMask(None=@n29.RGB) {g=true, r=true}) {materialFunction=&EncodeVolumeCoordinates}.Result
   Falloff Width=ScalarParameter() {defaultValue=0.02, parameterName=FW}
   Hardness=ScalarParameter() {parameterName=hardness}
   Falloff Power=ScalarParameter() {defaultValue=2, parameterName=FP}
   XY Frames=ScalarParameter() {defaultValue=10, parameterName=XYFrames}
  ) {materialFunction=/Volumetrics/Content/VolumeTextures/MaterialFunctions/VolumeEdgeMask}
  n31 = Constant() {r=1}
  n32 = Multiply(
   A=Multiply(
    A=Clamp(None=@n26) {maxDefault=40000, minDefault=-40000}
    B=MaterialFunctionCall(X=@n27, Y=@n27, Z=@n27, A=@n28) {materialFunction=&MakeFloat4}.Result
   )
   B=StaticSwitchParameter(True=Constant() {r=1}, False=StaticSwitchParameter(True=@n30.Result, False=@n31) {defaultValue=true, parameterName=Use Edge Mask}) {defaultValue=true, parameterName=IsVelocity}
  )
  ComponentMask(None=@n32) {b=true, g=true, r=true}
  OneMinus(None=ComponentMask(None=@n32) {a=true})
 ]
}
