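# Unlit material for a simple 3D fluid simulation advection pass on volumes tiled as 2D frame atlases.
# It traces back along Velocity by TimeStep to sample Source, then adds spherical emission and fades with a volume edge mask.
# IsVelocity blends between field types, adding alpha-driven Z buoyancy, centre attraction and temperature emission for velocity.
# Jump-flood distance and gradient samples lead only to dangling nodes.
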Material /Volumetrics/Tools/FluidSimulation/Materials/3D/M_Advect_3D_Simple
graph {
 properties {
  bAllowNegativeEmissiveColor=true
  blendMode=BLEND_AlphaComposite
  pixelDepthOffsetMode=PDOM_Legacy
  shadingModel=MSM_Unlit
  usedShadingModels=Unlit
 }
 nodes [
  n1 = MaterialFunctionCall!(
   UVs=TextureCoordinate()
   XY Frames=ComponentMask(None=VectorParameter() {parameterName=XYFrames, defaultValue=<rgba 16 16 0 0>}.RGB) {g=true, r=true}
  ) {materialFunction=/Volumetrics/Content/VolumeTextures/MaterialFunctions/EncodeVolumeCoordinates}
  n2 = Divide(B=ComponentMask(None=VectorParameter() {parameterName=FrameResolution, defaultValue=<rgba 256 256 0 0>}.RGB) {g=true, r=true}) {constA=0.5, desc=Half Texel Size}
  AppendVector(A=Clamp(None=ComponentMask(None=@n1.Result) {g=true, r=true}, Min=@n2, Max=OneMinus(None=@n2)), B=ComponentMask(None=@n1.Result) {b=true})
  n3 = ComponentMask(None=VectorParameter() {parameterName=XYFrames, defaultValue=<rgba 16 16 0 0>}.RGB) {g=true, r=true}
  n4 = MaterialFunctionCall(XY Frames=@n3) {materialFunction=&EncodeVolumeCoordinates}
  n5 = Multiply(
   A=ComponentMask(None=TextureSampleParameter2D() {desc=Velocity, parameterName=Velocity, texture=/Volumetrics/Tools/FluidSimulation/Textures/RT_Tiny}.RGB) {b=true, g=true, r=true}
   B=Divide(
    A=ScalarParameter() {defaultValue=5, desc="Default = 90 hz", parameterName=TimeStep}
    B=VectorParameter() {parameterName=FrameResolution, defaultValue=<rgba 256 256 256 0>}.RGB
   )
  )
  n6 = VectorParameter() {parameterName=FrameResolution, defaultValue=<rgba 256 256 256 0>}
  n7 = ComponentMask(None=Divide(A=Divide(B=@n6.RGB)) {desc=Half Texel}) {g=true, r=true}
  n8 = Constant()
  n9 = OneMinus(None=@n7)
  n10 = Divide(A=Subtract(A=@n6.B), B=@n6.B)
  n11 = MaterialFunctionCall!(
   Texture Object=TextureObjectParameter() {desc=Velocity, parameterName=Source, texture=&RT_Tiny}
   Position=Clamp(None=Subtract(A=@n4.Result, B=@n5) {constB=0}, Min=AppendVector(A=@n7, B=@n8) {desc=Min}, Max=AppendVector(A=@n9, B=@n10) {desc=Max})
   Scale=Constant() {r=1}
   Num Frames=@n3
   Clamp Z
  ) {materialFunction=/Volumetrics/Content/VolumeTextures/MaterialFunctions/VolumeTextureFunction}
  n12 = ScalarParameter() {defaultValue=5, parameterName=Temp Buoyancy}
  n13 = MaterialFunctionCall(
   UVs=TextureCoordinate()
   XY Frames=ComponentMask(None=VectorParameter() {parameterName=XYFrames, defaultValue=<rgba 16 16 0 0>}.RGB) {g=true, r=true}
  ) {materialFunction=&EncodeVolumeCoordinates}
  n14 = Multiply(A=Normalize(VectorInput=Subtract(A=@n13.Result) {constA=0.5, constB=0.5}), B=ScalarParameter() {defaultValue=0.2, parameterName=Attraction})
  n15 = Add(
   A=Multiply(A=Multiply(A=@n11.Result Alpha, B=@n12), B=Constant3Vector() {constant=<rgba 0 0 1 0>})
   B=Multiply(A=Multiply(A=@n14)) {constB=-1}
  )
  n16 = ScalarParameter() {parameterName=IsVelocity}
  n17 = AppendVector(A=Add(A=@n11.Result RGB, B=Multiply(A=@n15, B=@n16)) {constB=0}, B=Add(A=Multiply(A=@n11.Result Alpha)) {constB=0})
  n18 = MaterialFunctionCall(XY Frames=ComponentMask(None=VectorParameter() {parameterName=XYFrames, defaultValue=<rgba 16 16 0 0>}.RGB) {g=true, r=true}) {materialFunction=&EncodeVolumeCoordinates}
  n19 = VectorParameter() {parameterName=P2, defaultValue=<rgba 0.5 0.5 0.5 0>}
  n20 = ScalarParameter() {defaultValue=0.1, parameterName=R}
  n21 = OneMinus(None=MaterialFunctionCall!(
    Depth=SphereMask(A=@n18.Result, B=@n19.RGB, Radius=@n20)
    Density=ScalarParameter() {defaultValue=1, parameterName=Emission Falloff Density}
    Use Exp2
   ) {materialFunction=/Engine/Functions/Engine_MaterialFunctions01/Density/ExponentialDensity}.Result)
  n22 = Multiply(A=ScalarParameter() {defaultValue=1, parameterName=Emission Temperature}) {constB=0.01}
  n23 = LinearInterpolate(
   A=Multiply(A=@n21, B=ScalarParameter() {defaultValue=1, parameterName=Emission Strength})
   B=AppendVector(A=Constant3Vector(), B=Multiply(A=@n21, B=@n22))
   Alpha=ScalarParameter() {parameterName=IsVelocity}
  ) {constB=0}
  n24 = MaterialFunctionCall(XY Frames=ComponentMask(None=VectorParameter() {parameterName=XYFrames, defaultValue=<rgba 16 16 0 0>}.RGB) {g=true, r=true}) {materialFunction=&EncodeVolumeCoordinates}
  n25 = ScalarParameter() {defaultValue=0.02, parameterName=FW}
  n26 = ScalarParameter() {parameterName=hardness}
  n27 = ScalarParameter() {defaultValue=2, parameterName=FP}
  n28 = ScalarParameter() {defaultValue=10, parameterName=XYFrames}
  n29 = Multiply(
   A=Multiply(A=Add(A=@n17, B=@n23) {constB=0})
   B=LinearInterpolate(
    A=MaterialFunctionCall!(UVW=@n24.Result, Falloff Width=@n25, Hardness=@n26, Falloff Power=@n27, XY Frames=@n28) {materialFunction=/Volumetrics/Content/VolumeTextures/MaterialFunctions/VolumeEdgeMask}.Result
    Alpha=ScalarParameter() {parameterName=IsVelocity}
   ) {constA=1}
  )
  ComponentMask(None=@n29) {b=true, g=true, r=true}
  n30 = MaterialFunctionCall(
   Position=Multiply(A=@n1.Result)
   Scale=Constant() {r=1}
   Num Frames=ComponentMask(None=VectorParameter() {parameterName=SampleXYFrames, defaultValue=<rgba 16 8 0 0>}.RGB) {g=true, r=true}
   Clamp Z=StaticBool() {value=true}
  ) {materialFunction=&VolumeTextureFunction}
  n31 = TextureSampleParameter2D!(UVs=Multiply(A=@n30.UV 1), Level, Apply View MipBias) {
   constMipValue=0
   mipValueMode=TMVM_MipLevel
   parameterName=Gradient
   texture=/Volumetrics/Tools/JumpFlooding/Textures/RT_JumpFlood_3D_4k_Grad
  }
  n32 = TextureSampleParameter2D(UVs=@n30.UV 2) {
   constMipValue=0
   mipValueMode=TMVM_MipLevel
   parameterName=Gradient
   texture=&RT_JumpFlood_3D_4k_Grad
  }
  LinearInterpolate(A=@n31.RGB, B=@n32.RGB, Alpha=@n30.Z Phase)
  n33 = VectorParameter() {parameterName=XYFrames, defaultValue=<rgba 16 16 0 0>}
  MaterialFunctionCall(
   UVW=MaterialFunctionCall(XY Frames=ComponentMask(None=@n33.RGB) {g=true, r=true}) {materialFunction=&EncodeVolumeCoordinates}.Result
   Falloff Width=ScalarParameter() {defaultValue=0.2, parameterName=FW_1}
   Hardness=ScalarParameter() {defaultValue=0.1, parameterName=hardness_1}
   Falloff Power=ScalarParameter() {defaultValue=2, parameterName=FP_1}
   XY Frames=ScalarParameter() {defaultValue=10, parameterName=XYFrames_1}
  ) {materialFunction=&VolumeEdgeMask}
  Multiply()
  OneMinus(None=ComponentMask(None=@n29) {a=true})
  StaticSwitchParameter() {defaultValue=true, parameterName=IsVelocity}
  n34 = TextureSampleParameter2D(UVs=@n30.UV 1) {constMipValue=0, mipValueMode=TMVM_MipLevel, parameterName=DF, texture=/Volumetrics/Tools/JumpFlooding/Textures/RT_JumpFlood_3D_4k_DF}
  n35 = TextureSampleParameter2D(UVs=@n30.UV 2) {constMipValue=0, mipValueMode=TMVM_MipLevel, parameterName=DF, texture=&RT_JumpFlood_3D_4k_DF}
  Subtract(A=LinearInterpolate(A=@n34.RGB, B=@n35.RGB, Alpha=@n30.Z Phase))
  Subtract(A=@n11.Result RGB) {constB=0}
  VectorParameter() {parameterName=P, defaultValue=<rgba 0.5 0.5 0.8 1>}
 ]
}
