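# Material function utility for the fluid simulation tools that linearly interpolates six independent A/B value pairs at once.
# Each of the six outputs blends its A and B inputs by its own scalar alpha.
# A shared Enable static switch, on by default, selects between the blend and passing each A input straight through, so the disabled branch is resolved at compile time.
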MaterialFunction /Volumetrics/Tools/FluidSimulation/MaterialFunctions/Lerp6
graph {nodes [
  n1 = FunctionInput() {inputName=A1}
  n2 = FunctionInput() {inputName=B1, sortPriority=6}
  n3 = FunctionInput() {inputName=Alpha1, inputType=FunctionInput_Scalar, sortPriority=12}
  n4 = FunctionInput(Preview=StaticBool() {value=true}) {bUsePreviewValueAsDefault=true, inputName=Enable, inputType=FunctionInput_StaticBool, sortPriority=50}
  FunctionOutput(None=StaticSwitch(True=LinearInterpolate(A=@n1, B=@n2, Alpha=@n3), False=@n1, Value=@n4)) {outputName=1}
  n5 = FunctionInput() {inputName=A2, sortPriority=1}
  n6 = FunctionInput() {inputName=B2, sortPriority=7}
  n7 = FunctionInput() {inputName=Alpha2, inputType=FunctionInput_Scalar, sortPriority=13}
  FunctionOutput(None=StaticSwitch(True=LinearInterpolate(A=@n5, B=@n6, Alpha=@n7), False=@n5, Value=@n4)) {outputName=2, sortPriority=1}
  n8 = FunctionInput() {inputName=A3, sortPriority=2}
  n9 = FunctionInput() {inputName=B3, sortPriority=8}
  n10 = FunctionInput() {inputName=Alpha3, inputType=FunctionInput_Scalar, sortPriority=14}
  FunctionOutput(None=StaticSwitch(True=LinearInterpolate(A=@n8, B=@n9, Alpha=@n10), False=@n8, Value=@n4)) {outputName=3, sortPriority=2}
  n11 = FunctionInput() {inputName=A4, sortPriority=3}
  n12 = FunctionInput() {inputName=B4, sortPriority=9}
  n13 = FunctionInput() {inputName=Alpha4, inputType=FunctionInput_Scalar, sortPriority=15}
  FunctionOutput(None=StaticSwitch(True=LinearInterpolate(A=@n11, B=@n12, Alpha=@n13), False=@n11, Value=@n4)) {outputName=4, sortPriority=3}
  n14 = FunctionInput() {inputName=A5, sortPriority=4}
  n15 = FunctionInput() {inputName=B5, sortPriority=10}
  n16 = FunctionInput() {inputName=Alpha5, inputType=FunctionInput_Scalar, sortPriority=16}
  FunctionOutput(None=StaticSwitch(True=LinearInterpolate(A=@n14, B=@n15, Alpha=@n16), False=@n14, Value=@n4)) {outputName=5, sortPriority=4}
  n17 = FunctionInput() {inputName=A6, sortPriority=5}
  n18 = FunctionInput() {inputName=B6, sortPriority=11}
  n19 = FunctionInput() {inputName=Alpha6, inputType=FunctionInput_Scalar, sortPriority=17}
  FunctionOutput(None=StaticSwitch(True=LinearInterpolate(A=@n17, B=@n18, Alpha=@n19), False=@n17, Value=@n4)) {outputName=6, sortPriority=5}
 ]}
