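# Unlit material that computes velocity divergence for a 3D fluid simulation on volume textures laid out as frame grids.
# VolumeOffsetSamples reads neighbor samples from Velocity and Obstacle textures.
# Lerp6 blends velocity toward obstacle values by each neighbor's obstacle alpha, though all six value pairs are wired to -X samples.
# Central differences of the blended X, Y and Z components are summed and scaled by half the inverse GridScale.
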Material /Volumetrics/Tools/FluidSimulation/Materials/3D/M_Calculate_Divergence_3D
graph {
 properties {bAllowNegativeEmissiveColor=true, pixelDepthOffsetMode=PDOM_Legacy, shadingModel=MSM_Unlit, usedShadingModels=Unlit}
 nodes [
  n1 = TextureObjectParameter() {desc=Velocity, parameterName=Velocity, samplerType=SAMPLERTYPE_LinearColor, texture=/Volumetrics/Content/VolumeTextures/Textures/RT_Default}
  n2 = MaterialFunctionCall!(
   Texture Object=@n1
   Resolution=VectorParameter() {parameterName=Resolution, defaultValue=<rgba 4096 4096 0 0>}.RGB
   XYFrames=VectorParameter() {parameterName=XYFrames, defaultValue=<rgba 16 16 0 0>}.RGB
   Clamp Edges
  ) {materialFunction=/Volumetrics/Tools/FluidSimulation/MaterialFunctions/VolumeOffsetSamples}
  n3 = TextureObjectParameter() {desc=Velocity, parameterName=Obstacle, samplerType=SAMPLERTYPE_LinearColor, texture=&RT_Default}
  n4 = MaterialFunctionCall(
   Texture Object=@n3
   Resolution=VectorParameter() {parameterName=ObstacleResolution, defaultValue=<rgba 4096 4096 0 0>}.RGB
   XYFrames=VectorParameter() {parameterName=ObstacleXYFrames, defaultValue=<rgba 16 16 0 0>}.RGB
  ) {materialFunction=&VolumeOffsetSamples}
  n5 = MaterialFunctionCall!(
   A1=@n2.-X
   A2=@n2.-X
   A3=@n2.-X
   A4=@n2.-X
   A5=@n2.-X
   A6=@n2.-X
   B1=@n4.-X
   B2=@n4.-X
   B3=@n4.-X
   B4=@n4.-X
   B5=@n4.-X
   B6=@n4.-X
   Alpha1=ComponentMask(None=@n4.-X) {a=true}
   Alpha2=ComponentMask(None=@n4.+X) {a=true}
   Alpha3=ComponentMask(None=@n4.-Y) {a=true}
   Alpha4=ComponentMask(None=@n4.+Y) {a=true}
   Alpha5=ComponentMask(None=@n4.-Z) {a=true}
   Alpha6=ComponentMask(None=@n4.+Z) {a=true}
   Enable
  ) {materialFunction=/Volumetrics/Tools/FluidSimulation/MaterialFunctions/Lerp6}
  n6 = Subtract(A=ComponentMask(None=@n5.2) {r=true}, B=ComponentMask(None=@n5.1) {r=true})
  n7 = Subtract(A=ComponentMask(None=@n5.4) {g=true}, B=ComponentMask(None=@n5.3) {g=true})
  n8 = ComponentMask(None=@n5.6) {b=true}
  n9 = ComponentMask(None=@n5.5) {b=true}
  Multiply(
   A=Divide(B=ScalarParameter() {defaultValue=1, parameterName=GridScale}) {constA=0.5, desc=half rdx}
   B=Add(A=Add(A=@n6, B=@n7), B=Subtract(A=@n8, B=@n9))
  )
 ]
}
