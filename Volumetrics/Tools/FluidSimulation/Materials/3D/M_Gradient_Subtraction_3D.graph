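# Unlit material for the pressure gradient subtraction step of a 3D fluid simulation on volumes tiled into 2D frame atlases.
# VolumeOffsetSamples fetches axis-neighbor Pressure and Obstacle samples, and Lerp6 uses obstacle alpha to swap in center pressure at solid neighbors.
# The halved pressure differences, divided by GridScale, are subtracted from center Velocity.
# Per-axis obstacle masks then blend the affected velocity components toward values sampled from the Obstacle texture.
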Material /Volumetrics/Tools/FluidSimulation/Materials/3D/M_Gradient_Subtraction_3D
graph {
 properties {
  bAllowNegativeEmissiveColor=true
  blendMode=BLEND_AlphaComposite
  pixelDepthOffsetMode=PDOM_Legacy
  shadingModel=MSM_Unlit
  usedShadingModels=Unlit
 }
 nodes [
  Constant()
  LinearInterpolate()
  n1 = TextureSampleParameter2D(UVs=TextureCoordinate()) {
   desc=Velocity - Center Sample
   parameterName=Velocity
   samplerType=SAMPLERTYPE_LinearColor
   texture=/Volumetrics/Content/VolumeTextures/Textures/RT_Default
  }
  n2 = TextureObjectParameter() {desc=Velocity, parameterName=Pressure, samplerType=SAMPLERTYPE_LinearColor, texture=&RT_Default}
  n3 = MaterialFunctionCall!(
   Texture Object=@n2
   Resolution=VectorParameter() {parameterName=Resolution, defaultValue=<rgba 4096 4096 0 0>}.RGB
   XYFrames=VectorParameter() {parameterName=XYFrames, defaultValue=<rgba 16 16 0 0>}.RGB
   Clamp Edges
  ) {materialFunction=/Volumetrics/Tools/FluidSimulation/MaterialFunctions/VolumeOffsetSamples}
  n4 = TextureSampleParameter2D() {desc=Center, parameterName=Pressure, samplerType=SAMPLERTYPE_LinearColor, texture=&RT_Default}
  n5 = TextureObjectParameter() {desc=Velocity, parameterName=Obstacle, samplerType=SAMPLERTYPE_LinearColor, texture=&RT_Default}
  n6 = MaterialFunctionCall(
   Texture Object=@n5
   Resolution=VectorParameter() {parameterName=ObstacleResolution, defaultValue=<rgba 4096 4096 0 0>}.RGB
   XYFrames=VectorParameter() {parameterName=ObstacleXYFrames, defaultValue=<rgba 16 16 0 0>}.RGB
  ) {materialFunction=&VolumeOffsetSamples}
  n7 = ComponentMask(None=@n6.-X) {a=true}
  n8 = ComponentMask(None=@n6.+X) {a=true}
  n9 = ComponentMask(None=@n6.-Y) {a=true}
  n10 = ComponentMask(None=@n6.+Y) {a=true}
  n11 = ComponentMask(None=@n6.-Z) {a=true}
  n12 = ComponentMask(None=@n6.+Z) {a=true}
  n13 = MaterialFunctionCall!(
   A1=@n3.-X
   A2=@n3.-X
   A3=@n3.-X
   A4=@n3.-X
   A5=@n3.-X
   A6=@n3.-X
   B1=@n4.R
   B2=@n4.R
   B3=@n4.R
   B4=@n4.R
   B5=@n4.R
   B6=@n4.R
   Alpha1=@n7
   Alpha2=@n8
   Alpha3=@n9
   Alpha4=@n10
   Alpha5=@n11
   Alpha6=@n12
   Enable
  ) {materialFunction=/Volumetrics/Tools/FluidSimulation/MaterialFunctions/Lerp6}
  n14 = ComponentMask(None=@n13.2) {r=true}
  n15 = ComponentMask(None=@n13.1) {r=true}
  n16 = ComponentMask(None=@n13.4) {r=true}
  n17 = ComponentMask(None=@n13.3) {r=true}
  n18 = ComponentMask(None=@n13.6) {r=true}
  n19 = ComponentMask(None=@n13.5) {r=true}
  n20 = Multiply(
   A=Divide(B=ScalarParameter() {defaultValue=1, parameterName=GridScale}) {constA=0.5, desc=half rdx}
   B=MaterialFunctionCall(X=Subtract(A=@n14, B=@n15), Y=Subtract(A=@n16, B=@n17), Z=Subtract(A=@n18, B=@n19)) {materialFunction=&MakeFloat3}.Result
  )
  n21 = Max(A=@n7, B=@n8)
  n22 = Max(A=@n9, B=@n10)
  n23 = Max(A=@n11, B=@n12)
  n24 = MaterialFunctionCall(A1=@n6.-X, A2=@n6.-X, A3=@n6.-X, B1=@n6.-X, B2=@n6.-X, B3=@n6.-X, Alpha1=@n8, Alpha2=@n10, Alpha3=@n12) {materialFunction=&Lerp6}
  LinearInterpolate(
   A=Subtract(A=@n1.RGB, B=Multiply(A=@n20))
   B=MaterialFunctionCall(X=ComponentMask(None=@n24.1) {r=true}, Y=ComponentMask(None=@n24.2) {g=true}, Z=ComponentMask(None=@n24.3) {b=true}) {materialFunction=&MakeFloat3}.Result
   Alpha=Multiply(A=MaterialFunctionCall(X=@n21, Y=@n22, Z=@n23) {materialFunction=&MakeFloat3}.Result)
  )
  n25 = TextureSampleParameter2D() {desc=Center, parameterName=Obstacle, samplerType=SAMPLERTYPE_LinearColor, texture=&RT_Default}
  OneMinus(None=@n25.A)
 ]
}
