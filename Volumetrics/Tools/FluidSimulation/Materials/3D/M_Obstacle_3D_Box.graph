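# Unlit alpha-composite material that builds a box-shaped edge mask for 3D fluid simulation obstacles.
# EncodeVolumeCoordinates derives volume UVW from tiled XY frames, and VolumeEdgeMask turns it into a falloff shaped by width, hardness and power.
# Static switches toggle walls and ceiling/floor by pinning their axes to 0.5, away from the edges.
# XYFrames exists as both a vector and a scalar parameter, with different defaults.
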Material /Volumetrics/Tools/FluidSimulation/Materials/3D/M_Obstacle_3D_Box
graph {
 properties {
  bAllowNegativeEmissiveColor=true
  blendMode=BLEND_AlphaComposite
  pixelDepthOffsetMode=PDOM_Legacy
  shadingModel=MSM_Unlit
  usedShadingModels=Unlit
 }
 nodes [
  n1 = MaterialFunctionCall!(UVs, XY Frames=ComponentMask(None=VectorParameter() {parameterName=XYFrames, defaultValue=<rgba 16 16 0 0>}.RGB) {g=true, r=true}) {materialFunction=/Volumetrics/Content/VolumeTextures/MaterialFunctions/EncodeVolumeCoordinates}
  n2 = Constant() {r=0.5}
  n3 = StaticSwitchParameter(True=@n1.Result, False=AppendVector(A=AppendVector(A=@n2, B=@n2), B=ComponentMask(None=@n1.Result) {b=true})) {defaultValue=true, parameterName=Use Walls}
  n4 = AppendVector(A=ComponentMask(None=@n3) {g=true, r=true}, B=Constant() {r=0.5})
  Multiply(A=MaterialFunctionCall!(
    UVW=StaticSwitchParameter(True=@n3, False=@n4) {defaultValue=true, parameterName=Use Ceiling and Floor}
    Falloff Width=ScalarParameter() {defaultValue=0.05, parameterName=FW}
    Hardness=ScalarParameter() {defaultValue=1, parameterName=hardness}
    Falloff Power=ScalarParameter() {defaultValue=2, parameterName=FP}
    XY Frames=ScalarParameter() {defaultValue=10, parameterName=XYFrames}
   ) {materialFunction=/Volumetrics/Content/VolumeTextures/MaterialFunctions/VolumeEdgeMask}.Result)
 ]
}
