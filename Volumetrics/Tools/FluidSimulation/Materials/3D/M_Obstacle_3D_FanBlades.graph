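# Unlit alpha-composite material that procedurally draws a spinning 3D fan as an obstacle for the volumetric fluid simulation, using flipbook-encoded volume coordinates.
# Rotated, spiral-twisted coordinates become a repeating radial angle that sets each blade's gap, scooped height and thickness, clipped to the fan radius.
# Blades combine with a circular inlet wall into a solid boundary stored as inverse alpha.
# A rotation-rate offset computes signed blade velocity, with negative emissive allowed.
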Material /Volumetrics/Tools/FluidSimulation/Materials/3D/M_Obstacle_3D_FanBlades
graph {
 properties {
  bAllowNegativeEmissiveColor=true
  blendMode=BLEND_AlphaComposite
  pixelDepthOffsetMode=PDOM_Legacy
  shadingModel=MSM_Unlit
  usedShadingModels=Unlit
 }
 nodes [
  n1 = ScalarParameter() {parameterName=Rot}
  n2 = MaterialFunctionCall!(UVs, XY Frames=ComponentMask(None=VectorParameter() {parameterName=XYFrames, defaultValue=<rgba 16 16 0 0>}.RGB) {g=true, r=true}) {materialFunction=/Volumetrics/Content/VolumeTextures/MaterialFunctions/EncodeVolumeCoordinates}
  n3 = Multiply(
   A=ScalarParameter() {defaultValue=-0.1, parameterName=Fan Blade Spiral}
   B=SphereMask(A=ComponentMask(None=@n2.Result) {g=true, r=true}, B=Constant() {r=0.5}, Radius=Constant() {r=0.5}) {hardnessPercent=0}
  )
  n4 = Subtract(A=@n2.Result) {constB=0.5}
  n5 = Add(
   A=RotateAboutAxis(NormalizedRotationAxis=Constant3Vector() {constant=<rgba 0 0 1 0>}, RotationAngle=Add(A=@n1, B=@n3), PivotPoint=Constant(), Position=@n4)
   B=@n4
  )
  Add(A=@n5) {constB=0.5}
  n6 = MaterialFunctionCall(XY Frames=ComponentMask(None=VectorParameter() {parameterName=XYFrames_1, defaultValue=<rgba 16 16 0 0>}.RGB) {g=true, r=true}) {materialFunction=&EncodeVolumeCoordinates}
  n7 = ComponentMask(None=@n6.Result) {g=true, r=true}
  n8 = Constant() {r=0.5}
  Multiply(A=MaterialFunctionCall!(
    UVW=AppendVector(A=@n7, B=@n8)
    Falloff Width=ScalarParameter() {defaultValue=0.02, parameterName=FW}
    Hardness=ScalarParameter() {defaultValue=1, parameterName=hardness}
    Falloff Power=ScalarParameter() {defaultValue=2, parameterName=FP}
    XY Frames=ScalarParameter() {defaultValue=10, parameterName=XYFrames}
   ) {materialFunction=/Volumetrics/Content/VolumeTextures/MaterialFunctions/VolumeEdgeMask}.Result)
  n9 = ComponentMask(None=@n5) {g=true, r=true}
  n10 = MaterialFunctionCall(Vector or UVs=Normalize(VectorInput=@n9)) {materialFunction=&VectorToRadialValue}
  n11 = ScalarParameter() {defaultValue=6, desc=Number of Blades, parameterName=Blades}
  n12 = MaterialFunctionCall!(
   Alpha=Frac(None=Multiply(A=@n10.Vector Converted to Angle, B=@n11))
   Offset=ScalarParameter() {defaultValue=0.5, desc=Amount of Gap between each Blade, parameterName=BladeGap}
  ) {materialFunction=/Engine/Functions/Engine_MaterialFunctions01/Opacity/AlphaOffset}
  n13 = Saturate(None=@n12.Result)
  n14 = ScalarParameter() {defaultValue=1.5, desc="Gives blades a curved profile. 1 = Linear", parameterName=ScoopPower}
  n15 = Multiply(
   A=Subtract(A=Power(Base=@n13, Exp=@n14)) {constB=0.5}
   B=ScalarParameter() {defaultValue=0.2, desc="The height of each Blade, changes the Angle of the Blades", parameterName=Blade Height}
  )
  n16 = Ceil(None=@n12.Result)
  n17 = ComponentMask(None=@n2.Result) {b=true}
  n18 = VectorParameter() {parameterName=XYFrames, defaultValue=<rgba 16 16 0 0>}
  n19 = ComponentMask(None=MaterialFunctionCall(XY Frames=ComponentMask(None=@n18.RGB) {g=true, r=true}) {materialFunction=&EncodeVolumeCoordinates}.Result) {g=true, r=true}
  n20 = MaterialFunctionCall(Vector 2=Multiply(A=Subtract(B=@n19) {constA=0.5, constB=0.5}) {constB=2}) {materialFunction=/Engine/Functions/Engine_MaterialFunctions02/Utility/VectorLength}
  n21 = ScalarParameter() {defaultValue=0.85, parameterName=FanRadius}
  n22 = Floor(None=Add(A=@n20.V2 Length, B=OneMinus(None=@n21)))
  n23 = Multiply(
   A=SphereMask(
    A=Add(A=@n15) {constB=0.5}
    B=Multiply(A=@n16, B=@n17)
    Radius=ScalarParameter() {defaultValue=0.02, desc=How thick the blades are, parameterName=BladeThickness}
   )
   B=OneMinus(None=Saturate(None=@n22))
  ) {desc=Fan Blades * Fan Radius}
  n24 = Constant3Vector() {constant=<rgba 0 0 1 0>}
  n25 = ScalarParameter() {defaultValue=0.05, parameterName=RotRate}
  n26 = Constant()
  Multiply(A=@n23, B=Multiply(A=RotateAboutAxis(NormalizedRotationAxis=@n24, RotationAngle=@n25, PivotPoint=@n26, Position=@n4)) {constB=-256})
  Multiply(A=@n6.Result, B=Constant3Vector() {constant=<rgba 1 1 0.9 0>})
  n27 = ScalarParameter() {defaultValue=0.95, parameterName=InletRadius}
  n28 = Floor(None=Add(A=@n20.V2 Length, B=OneMinus(None=@n27)))
  OneMinus(None=Max(A=@n23, B=Saturate(None=@n28)) {desc=Solid Boundary from Walls and Fan Blades}) {desc=Alpha is stored as inverse alpha}
  Subtract(A=@n9) {constB=0.5}
 ]
}
