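# M_Obstacle_Empty is an empty unlit material in the 3D fluid simulation tools.
# Its graph holds no expressions, so every output stays at its default.
# The name suggests it serves as a blank obstacle input for the simulation.
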Material /Volumetrics/Tools/FluidSimulation/Materials/3D/M_Obstacle_Empty
graph {expressions {}, properties {pixelDepthOffsetMode=PDOM_Legacy, shadingModel=MSM_Unlit, usedShadingModels=Unlit}}
