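# Unlit material for one iteration of a 3D pressure solve in a fluid simulation, on volumes packed as 2D frame atlases.
# It combines Divergence, scaled by negative GridScale squared, with six pressure terms and divides by NumSamples, Jacobi-style.
# Obstacle alpha blends each term toward center pressure, though all six take the -X pressure sample as input.
# Divisor scales sampling resolution, and its inverse weights the Z terms against center pressure.
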Material /Volumetrics/Tools/FluidSimulation/Materials/3D/M_Pressure_Iteration_3D
graph {
 properties {bAllowNegativeEmissiveColor=true, pixelDepthOffsetMode=PDOM_Legacy, shadingModel=MSM_Unlit, usedShadingModels=Unlit}
 nodes [
  n1 = ScalarParameter() {defaultValue=1, parameterName=GridScale}
  n2 = Multiply(A=Multiply(A=@n1, B=@n1)) {constB=-1}
  n3 = VectorParameter() {parameterName=XYFrames, defaultValue=<rgba 16 16 0 0>}
  n4 = MaterialFunctionCall!(
   Texture Object=TextureObjectParameter() {parameterName=Divergence, samplerType=SAMPLERTYPE_LinearColor, texture=/Volumetrics/Content/VolumeTextures/Textures/RT_Default}
   Position=MaterialFunctionCall!(UVs, XY Frames=ComponentMask(None=@n3.RGB) {g=true, r=true}) {materialFunction=/Volumetrics/Content/VolumeTextures/MaterialFunctions/EncodeVolumeCoordinates}.Result
   Scale=Constant() {r=1}
   Num Frames=ComponentMask(None=VectorParameter() {parameterName=SampleXYFrames, defaultValue=<rgba 16 16 0 0>}.RGB) {g=true, r=true}
  ) {materialFunction=/Volumetrics/Content/VolumeTextures/MaterialFunctions/VolumeTextureFunction_NoZBlend}
  n5 = TextureObjectParameter() {desc=Velocity, parameterName=Pressure, samplerType=SAMPLERTYPE_LinearColor, texture=&RT_Default}
  n6 = VectorParameter() {parameterName=Resolution, defaultValue=<rgba 4096 4096 0 0>}
  n7 = ScalarParameter() {defaultValue=1, parameterName=Divisor}
  n8 = MaterialFunctionCall!(
   Texture Object=@n5
   Resolution=Multiply(A=@n6.RGB, B=@n7)
   XYFrames=VectorParameter() {parameterName=XYFrames, defaultValue=<rgba 16 16 0 0>}.RGB
   Clamp Edges
  ) {materialFunction=/Volumetrics/Tools/FluidSimulation/MaterialFunctions/VolumeOffsetSamples}
  n9 = TextureSampleParameter2D() {desc=Center, parameterName=Pressure, samplerType=SAMPLERTYPE_LinearColor, texture=&RT_Default}
  n10 = TextureObjectParameter() {desc=Velocity, parameterName=Obstacle, samplerType=SAMPLERTYPE_LinearColor, texture=&RT_Default}
  n11 = MaterialFunctionCall(
   Texture Object=@n10
   Resolution=@n6.RGB
   XYFrames=ComponentMask(None=VectorParameter() {parameterName=XYFrames, defaultValue=<rgba 16 16 0 0>}.RGB) {g=true, r=true}
  ) {materialFunction=&VolumeOffsetSamples}
  n12 = MaterialFunctionCall!(
   A1=@n8.-X
   A2=@n8.-X
   A3=@n8.-X
   A4=@n8.-X
   A5=@n8.-X
   A6=@n8.-X
   B1=@n9.R
   B2=@n9.R
   B3=@n9.R
   B4=@n9.R
   B5=@n9.R
   B6=@n9.R
   Alpha1=ComponentMask(None=@n11.-X) {a=true}
   Alpha2=ComponentMask(None=@n11.+X) {a=true}
   Alpha3=ComponentMask(None=@n11.-Y) {a=true}
   Alpha4=ComponentMask(None=@n11.+Y) {a=true}
   Alpha5=ComponentMask(None=@n11.-Z) {a=true}
   Alpha6=ComponentMask(None=@n11.+Z) {a=true}
   Enable
  ) {materialFunction=/Volumetrics/Tools/FluidSimulation/MaterialFunctions/Lerp6}
  n13 = Add(
   A=Add(A=ComponentMask(None=@n12.1) {r=true}, B=ComponentMask(None=@n12.2) {r=true})
   B=Add(A=ComponentMask(None=@n12.3) {r=true}, B=ComponentMask(None=@n12.4) {r=true})
  )
  n14 = Divide(B=@n7)
  n15 = Add(
   A=LinearInterpolate(A=@n9.R, B=ComponentMask(None=@n12.5) {r=true}, Alpha=@n14)
   B=LinearInterpolate(A=@n9.R, B=ComponentMask(None=@n12.6) {r=true}, Alpha=@n14)
  )
  Divide(A=Add(A=Multiply(A=@n2, B=@n4.Result), B=Add(A=@n13, B=@n15)), B=ScalarParameter() {defaultValue=6, parameterName=NumSamples})
  n16 = TextureCoordinate()
  n17 = VectorParameter() {parameterName=UVScaleAndOffset, defaultValue=<rgba 1 1 0 0>}
  n18 = ComponentMask(None=@n17.RGB) {g=true, r=true}
  TextureSampleParameter2D(UVs=Add(A=Multiply(A=@n16, B=@n18), B=AppendVector(A=@n17.B, B=@n17.B))) {desc=Center, parameterName=Divergence, texture=&RT_Default}
  VectorParameter() {parameterName=ObstacleResolution, defaultValue=<rgba 256 256 256 0>}
  VectorParameter() {parameterName=ObstacleXYFrames, defaultValue=<rgba 16 16 0 0>}
 ]
}
