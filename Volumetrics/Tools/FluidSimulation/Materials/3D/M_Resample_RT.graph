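# Unlit alpha-composite material that resamples a flipbook-layout volume render target into another frame layout, for the fluid simulation tools.
# It encodes output UVs as volume coordinates, clamps XY by half a texel, and maps them to two source slices.
# Both slices are sampled at scaled UVs and blended by Z phase for color and inverted alpha.
# The graph also carries debug digit readouts, one added to the dimmed color.
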Material /Volumetrics/Tools/FluidSimulation/Materials/3D/M_Resample_RT
graph {
 properties {
  bAllowNegativeEmissiveColor=true
  blendMode=BLEND_AlphaComposite
  pixelDepthOffsetMode=PDOM_Legacy
  shadingModel=MSM_Unlit
  usedShadingModels=Unlit
 }
 nodes [
  n1 = MaterialFunctionCall!(
   UVs=TextureCoordinate()
   XY Frames=ComponentMask(None=VectorParameter() {parameterName=XYFrames, defaultValue=<rgba 16 16 0 0>}.RGB) {g=true, r=true}
  ) {materialFunction=/Volumetrics/Content/VolumeTextures/MaterialFunctions/EncodeVolumeCoordinates}
  n2 = ComponentMask(None=@n1.Result) {g=true, r=true}
  n3 = Divide(B=ComponentMask(None=VectorParameter() {parameterName=VolumeResolution, defaultValue=<rgba 256 256 0 0>}.RGB) {g=true, r=true}) {constA=0.5, desc=Half Texel Size}
  n4 = OneMinus(None=@n3)
  n5 = ComponentMask(None=@n1.Result) {b=true}
  n6 = MaterialFunctionCall!(
   Texture Object
   Position=AppendVector(A=Clamp(None=@n2, Min=@n3, Max=@n4), B=@n5)
   Scale=Constant() {r=1}
   Num Frames=ComponentMask(None=VectorParameter() {parameterName=SampleXYFrames, defaultValue=<rgba 8 8 0 0>}.RGB) {g=true, r=true}
   Clamp Z=StaticBool() {value=true}
  ) {materialFunction=/Volumetrics/Content/VolumeTextures/MaterialFunctions/VolumeTextureFunction}
  n7 = ComponentMask(None=VectorParameter() {parameterName=UVScale, defaultValue=<rgba 0.125 0.125 0 0>}.RGB) {g=true, r=true}
  n8 = TextureSampleParameter2D!(UVs=Multiply(A=@n6.UV 1, B=@n7), Level, Apply View MipBias) {
   constMipValue=0
   mipValueMode=TMVM_MipLevel
   parameterName=RT
   samplerType=SAMPLERTYPE_LinearColor
   texture=/Volumetrics/Content/VolumeTextures/Textures/RT_Default
  }
  n9 = TextureSampleParameter2D(UVs=Multiply(A=@n6.UV 2, B=@n7) {constB=0.125}) {
   constMipValue=0
   mipValueMode=TMVM_MipLevel
   parameterName=RT
   samplerType=SAMPLERTYPE_LinearColor
   texture=&RT_Default
  }
  n10 = Multiply(A=TextureCoordinate(), B=ComponentMask(None=VectorParameter() {parameterName=XYFrames, defaultValue=<rgba 16 16 0 0>}.RGB) {g=true, r=true})
  n11 = VectorParameter() {parameterName=asdfa, defaultValue=<rgba 0 -0.5 1 2>}
  n12 = Multiply(A=Add(A=Frac(None=@n10), B=ComponentMask(None=@n11.RGB) {g=true, r=true}), B=AppendVector(A=@n11.B, B=@n11.B))
  Add(
   A=Multiply(A=LinearInterpolate(A=@n8.RGB, B=@n9.RGB, Alpha=@n6.Z Phase)) {constB=0.75}
   B=MaterialFunctionCall!(Vector3=@n6.z frame, MaximumNumberOfDigits, UVs=@n12, DebugTextLocation RG_UpperRight BA_LowerLeft, Component Spacing) {materialFunction=/Engine/Functions/Engine_MaterialFunctions02/Utility/DebugFloat3Values}.ColorCodedOutput
  )
  n13 = VectorParameter() {parameterName=XYFrames, defaultValue=<rgba 16 16 0 0>}
  n14 = Divide(B=Multiply(A=@n13.R, B=@n13.R)) {constA=0.5}
  MaterialFunctionCall(Number=Multiply(A=Add(A=@n5, B=@n14)) {constA=1, constB=8}, UVs=@n12) {materialFunction=&DebugScalarValues}
  OneMinus(None=LinearInterpolate(A=@n8.A, B=@n9.A, Alpha=@n6.Z Phase))
 ]
}
