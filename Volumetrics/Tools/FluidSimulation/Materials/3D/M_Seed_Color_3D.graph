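# Unlit alpha-composite material that seeds color for the 3D fluid simulation, working on a volume laid out as an XY frame grid.
# EncodeVolumeCoordinates turns the flat layout into 3D positions for a sphere distance falloff around point P, perturbed by inverted Voronoi noise and scaled into density.
# That density, capped, masks a heatmap gradient texture sampled by the volume coordinates' XY.
# Side branches compute an inset temperature ratio and an inverted sphere mask.
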Material /Volumetrics/Tools/FluidSimulation/Materials/3D/M_Seed_Color_3D
graph {
 properties {
  bAllowNegativeEmissiveColor=true
  blendMode=BLEND_AlphaComposite
  pixelDepthOffsetMode=PDOM_Legacy
  shadingModel=MSM_Unlit
  usedShadingModels=Unlit
 }
 nodes [
  n1 = MaterialFunctionCall!(UVs, XY Frames=ComponentMask(None=VectorParameter() {parameterName=XYFrames, defaultValue=<rgba 16 16 0 0>}.RGB) {g=true, r=true}) {materialFunction=/Volumetrics/Content/VolumeTextures/MaterialFunctions/EncodeVolumeCoordinates}
  n2 = ComponentMask(None=VectorParameter() {parameterName=XYFrames, defaultValue=<rgba 16 16 0 0>}.RGB) {g=true, r=true}
  n3 = Noise!(
   World Position=Multiply(
    A=MaterialFunctionCall(XY Frames=@n2) {materialFunction=&EncodeVolumeCoordinates}.Result
    B=ScalarParameter() {defaultValue=5, parameterName=Color Voronoi Scale}
   )
   FilterWidth
  ) {bTurbulence=false, levels=3, noiseFunction=NOISEFUNCTION_VoronoiALU, outputMin=0, quality=4}
  n4 = Multiply(A=ScalarParameter() {defaultValue=10, parameterName=Noise Strength}) {constB=0.1}
  n5 = ScalarParameter() {defaultValue=100, parameterName=Hardness}
  n6 = Add(A=Multiply(A=OneMinus(None=@n3), B=Multiply(A=@n4, B=@n5)), B=ScalarParameter() {parameterName=Offset})
  n7 = MaterialFunctionCall(XY Frames=ComponentMask(None=VectorParameter() {parameterName=XYFrames, defaultValue=<rgba 16 16 0 0>}.RGB) {g=true, r=true}) {materialFunction=&EncodeVolumeCoordinates}
  n8 = VectorParameter() {parameterName=P, defaultValue=<rgba 0.5 0.5 0.2 0>}
  n9 = Multiply(
   A=Subtract(A=Distance(A=@n7.Result, B=@n8.RGB), B=ScalarParameter() {defaultValue=0.1, parameterName=SourceRadius})
   B=Multiply(A=@n5) {constB=-1}
  )
  n10 = Max(A=Multiply(A=Add(A=@n6, B=@n9), B=ScalarParameter() {defaultValue=1, parameterName=Density Multiplier})) {constB=0}
  Multiply(
   A=TextureSample(UVs=ComponentMask(None=@n1.Result) {g=true, r=true}) {texture=/Engine/EngineDebugMaterials/HeatmapGradient}.RGB
   B=Min(A=@n10)
  )
  OneMinus(None=SphereMask(A=@n7.Result, B=@n8.RGB, Radius=ScalarParameter() {defaultValue=0.1, parameterName=SourceRadius}) {hardnessPercent=0})
  n11 = ScalarParameter() {defaultValue=1, parameterName=Heat Inset}
  n12 = Multiply(A=Max(A=Subtract(A=@n10, B=@n11)) {constB=0}, B=ScalarParameter() {defaultValue=1, parameterName=Temperature Multiplier})
  n13 = ScalarParameter() {defaultValue=1, parameterName=Density Multiplier}
  OneMinus(None=Clamp(None=Divide(A=@n12, B=@n13)))
  ScalarParameter() {defaultValue=0.1, parameterName=R}
 ]
}
