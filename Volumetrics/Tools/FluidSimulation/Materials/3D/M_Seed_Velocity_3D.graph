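# Unlit alpha-composite material that generates seed velocity for a 3D fluid simulation, using volume coordinates decoded from a 2D frame-tiled layout.
# Two curl noise layers, scaled by strength and overall force, are added to a sphere-masked radial push around point P and a weighted seed texture.
# Separate branches multiply the curl force by a volume edge mask and invert the Color texture's blue channel.
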Material /Volumetrics/Tools/FluidSimulation/Materials/3D/M_Seed_Velocity_3D
graph {
 properties {
  bAllowNegativeEmissiveColor=true
  blendMode=BLEND_AlphaComposite
  pixelDepthOffsetMode=PDOM_Legacy
  shadingModel=MSM_Unlit
  usedShadingModels=Unlit
 }
 nodes [
  n1 = VectorParameter() {parameterName=P, defaultValue=<rgba 0.5 0.5 0.5 0>}
  n2 = MaterialFunctionCall!(UVs, XY Frames=ComponentMask(None=VectorParameter() {parameterName=XYFrames, defaultValue=<rgba 16 16 0 0>}.RGB) {g=true, r=true}) {materialFunction=/Volumetrics/Content/VolumeTextures/MaterialFunctions/EncodeVolumeCoordinates}
  n3 = Multiply(
   A=SphereMask(A=@n1.RGB, B=@n2.Result, Radius=Constant() {r=0.2}) {hardnessPercent=50}
   B=Normalize(VectorInput=Subtract(A=@n2.Result, B=@n1.RGB) {constB=0.5})
  )
  n4 = ScalarParameter() {parameterName=Radial Force}
  n5 = Add(A=@n2.Result, B=VectorParameter() {parameterName=Offset}.RGB)
  n6 = VectorNoise(World Position=Multiply(A=@n5, B=ScalarParameter() {defaultValue=25, parameterName=Curl 1 Tiling})) {noiseFunction=VNF_CurlALU}
  n7 = ScalarParameter() {defaultValue=1, parameterName=Curl 1 Strength}
  n8 = VectorNoise(World Position=Multiply(A=@n5, B=ScalarParameter() {defaultValue=25, parameterName=Curl 2 Tiling})) {noiseFunction=VNF_CurlALU}
  n9 = ScalarParameter() {parameterName=Curl 2 Strength}
  n10 = Multiply(A=Add(A=Multiply(A=@n6, B=@n7), B=Multiply(A=@n8, B=@n9)), B=ScalarParameter() {defaultValue=1, parameterName=Force})
  Add(
   A=Multiply(
    A=TextureSample() {samplerType=SAMPLERTYPE_LinearColor, texture=/Volumetrics/Content/VolumeTextures/Textures/RT_Default}.RGB
    B=ScalarParameter() {parameterName=SeedTex}
   )
   B=Add(A=Multiply(A=@n3, B=@n4), B=@n10)
  )
  n11 = ComponentMask(None=VectorParameter() {parameterName=XYFrames, defaultValue=<rgba 16 16 0 0>}.RGB) {g=true, r=true}
  Multiply(
   A=MaterialFunctionCall!(
    UVW=MaterialFunctionCall(XY Frames=@n11) {materialFunction=&EncodeVolumeCoordinates}.Result
    Falloff Width=ScalarParameter() {defaultValue=0.02, parameterName=Edge Falloff Width}
    Hardness=ScalarParameter() {parameterName=Edge Falloff Hardness}
    Falloff Power
    XY Frames=@n11
   ) {materialFunction=/Volumetrics/Content/VolumeTextures/MaterialFunctions/VolumeEdgeMask}.Result
   B=@n10
  )
  OneMinus(None=TextureSampleParameter2D() {parameterName=Color, samplerType=SAMPLERTYPE_LinearColor, texture=&RT_Default}.B) {desc=Copy Temp Directly}
 ]
}
